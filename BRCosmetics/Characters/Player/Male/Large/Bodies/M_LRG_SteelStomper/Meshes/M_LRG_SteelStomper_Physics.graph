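# Physics asset for the SteelStomper large male player body cosmetic, covering its dangling gear: a hammer, wrench, pouch and two tools.
# Capsules wrap each dyn_ gear bone except the box-shaped hammer, with pelvis, right thigh and right calf capsules as body shapes.
# Angular-limited constraints chain hammer to strap to strap base, join pouch and tools to their bases, and tie the wrench to the pelvis.
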PhysicsAsset /BRCosmetics/Characters/Player/Male/Large/Bodies/M_LRG_SteelStomper/Meshes/M_LRG_SteelStomper_Physics
bodies {
 calf_r [{shapeType=Capsule, length=31.5, radius=5.4, shapeName=calf_r_capsule, center=<17.3637 1.7774 0>, rotation=<rot 90 83.2785 83.2788>}]
 dyn_hammer_r [{
   shapeType=Box
   extentX=19.5
   extentY=22.5
   extentZ=3.7
   shapeName=dyn_hammer_r_box
   center=<7.8896 0.359 -0.2781>
   rotation=<rot -0.0001 0.0001 0>
  }]
 dyn_hammer_strap_base_r [{shapeType=Capsule, length=0, radius=2.2, shapeName=dyn_hammer_strap_base_r_capsule}]
 dyn_hammer_strap_r [{
   shapeType=Capsule
   length=4.7
   radius=2.4
   shapeName=dyn_hammer_strap_r_capsule
   center=<3.7186 -0.044 0>
   rotation=<rot 80 -84.9993 -179.9998>
  }]
 dyn_pouch_base_l [{shapeType=Capsule, length=0, radius=2.3, shapeName=dyn_pouch_base_l_capsule}]
 dyn_pouch_l [{shapeType=Capsule, length=7, radius=3.1, shapeName=dyn_pouch_l_capsule, center=<1.7961 1.7375 0.1399>, rotation=<rot -0.0001 0 -0.0002>}]
 dyn_tool_1_r [{shapeType=Capsule, length=15, radius=1.4, shapeName=dyn_tool_1_r_capsule, center=<0 -0.5016 -1.0331>, rotation=<rot 0.0002 0 4.9998>}]
 dyn_tool_2_r [{shapeType=Capsule, length=15, radius=1.4, shapeName=dyn_tool_2_r_capsule, center=<0 -0.5016 -1.0331>, rotation=<rot 0.0002 0 4.9997>}]
 dyn_tool_base_1_r [{shapeType=Capsule, length=0.5, radius=1.7, shapeName=dyn_tool_base_1_r_capsule}]
 dyn_tool_base_2_r [{shapeType=Capsule, length=0.5, radius=1.7, shapeName=dyn_tool_base_2_r_capsule}]
 dyn_wrench_r [{
   shapeType=Capsule
   length=11
   radius=3.1
   shapeName=dyn_wrench_r_capsule
   center=<-0.2886 0.4386 1.2706>
   rotation=<rot -4.8287 1.2972 -15.0553>
  }]
 pelvis [{shapeType=Capsule, length=21, radius=9.1, shapeName=pelvis_capsule, center=<-12.5793 -3.0856 0>}]
 thigh_r [{shapeType=Capsule, length=31.3, radius=8.5, shapeName=thigh_r_capsule, center=<22.0391 -1.8197 -1.1562>, rotation=<rot 90 160.5289 160.5292>}]
}
constraints [
 {bone1Name=dyn_wrench_r, bone2Name=pelvis, swing1LimitDegrees=40, swing2LimitDegrees=25, twistLimitDegrees=10}
 {bone1Name=dyn_hammer_r, bone2Name=dyn_hammer_strap_r, swing1LimitDegrees=45, swing2LimitDegrees=45, twistLimitDegrees=30}
 {bone1Name=dyn_tool_2_r, bone2Name=dyn_tool_base_2_r, swing1LimitDegrees=10, swing2LimitDegrees=20, twistLimitDegrees=5}
 {bone1Name=dyn_tool_1_r, bone2Name=dyn_tool_base_1_r, swing1LimitDegrees=10, swing2LimitDegrees=20, twistLimitDegrees=5}
 {bone1Name=dyn_pouch_l, bone2Name=dyn_pouch_base_l, swing1LimitDegrees=15, swing2LimitDegrees=20, twistLimitDegrees=20}
 {bone1Name=dyn_hammer_strap_r, bone2Name=dyn_hammer_strap_base_r, swing1LimitDegrees=30, swing2LimitDegrees=30, twistLimitDegrees=30}
]
