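# Post-process-style cosmetic AnimBlueprint for the large male SteelStomper body, adding secondary motion to the pose copied from its attached parent mesh.
# A Control Rig receives a fly amount set by skydiving and floating state.
# Additive bone offsets on shoulder armor and hammer strap apply only in the front end.
# A RigidBody node then simulates the body's physics asset in pelvis space with overridden gravity and clamped acceleration.
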AnimBlueprint /BRCosmetics/Characters/Player/Male/Large/Bodies/M_LRG_SteelStomper/Meshes/M_LRG_SteelStomper_AnimBP
target_skeleton=&/M_LRG_SteelStomper_Skeleton
graphs [
 AnimGraph {nodes [
   n1 = Control Rig!(Source=Copy Pose From Mesh() {bUseAttachedParent=true}, Alpha, Fly_mode:double=Get Dyn_fly() -> (Dyn_fly:float)) -> (Pose) {
    alphaBoolBlend {blendOption=Linear}
    controlRigAssetReference {blueprintRigClass=&/M_LRG_SteelStomper_CtrlRig.M_LRG_SteelStomper_CtrlRig_C}
   }
   n2 = Get Dyn_Frontend_alfa() -> (Dyn_Frontend_alfa:float)
   n3 = "Transform (Modify) Bone - Bone: dyn_shoulder_armor_r"(ComponentPose=Local To Component(LocalPose=@n1), Alpha=@n2, Translation="-1,0,-1", Rotation="0,-7.5,-15", Scale="1,1,1") {
    rotationMode=BMM_Additive
    rotationSpace=BCS_BoneSpace
    translationMode=BMM_Additive
    translationSpace=BCS_BoneSpace
    alphaBoolBlend {blendOption=Linear}
    boneToModify {boneName=dyn_shoulder_armor_r}
    rotation {roll=-15, yaw=-7.5}
    translation {x=-1, z=-1}
   }
   n4 = "Transform (Modify) Bone - Bone: dyn_hammer_strap_base_r"(ComponentPose=@n3, Alpha=@n2, Translation="0,0,0", Rotation="-10,0,0", Scale="1,1,1") {
    rotationMode=BMM_Additive
    rotationSpace=BCS_BoneSpace
    translationMode=BMM_Additive
    translationSpace=BCS_BoneSpace
    alphaBoolBlend {blendOption=Linear}
    boneToModify {boneName=dyn_hammer_strap_base_r}
    rotation {pitch=-10}
   }
   n5 = "Transform (Modify) Bone - Bone: dyn_shoulder_armor_l"(ComponentPose=@n4, Alpha=@n2, Translation="0,1,0", Rotation="0,0,-2.5", Scale="1,1,1") {
    rotationMode=BMM_Additive
    rotationSpace=BCS_BoneSpace
    translationMode=BMM_Additive
    translationSpace=BCS_BoneSpace
    alphaBoolBlend {blendOption=Linear}
    boneToModify {boneName=dyn_shoulder_armor_l}
    rotation {roll=-2.5}
    translation {y=1}
   }
   n6 = RigidBody(ComponentPose=@n5) {
    bOverrideWorldGravity=true
    overridePhysicsAsset=&/M_LRG_SteelStomper_Physics
    simulationSpace=BaseBoneSpace
    alphaBoolBlend {blendOption=Linear}
    alphaScaleBiasClamp {bInterpResult=true}
    baseBoneRef {boneName=pelvis}
    componentAppliedLinearAccClamp=<850 850 850>
    componentLinearAccScale=<1 1 1>
    componentLinearVelScale=<1 1 1>
    overrideWorldGravity {z=-980}
    simSpaceSettings {maxAngularAcceleration=750, maxAngularVelocity=8, maxLinearAcceleration=250, maxLinearVelocity=10, velocityScaleZ=0.25, worldAlpha=1}
   }
   Output Pose(Result=Component To Local(ComponentPose=@n6)) {layerGroup=None}
  ]}
 EventGraph {nodes [
   Event BlueprintUpdateAnimation()
   > n1 = Sequence()
   > n2 = Branch(Condition=Get bIsSkyDiving())
   > n3 = Branch(Condition=Get bIsSkydiveFloating())
   > Set Dyn_fly(Dyn_fly:float=0.75) -> (Output_Get:float)
   Set Dyn_fly(execute=@n3.else, Dyn_fly=1)
   Set Dyn_fly(execute=@n2.else, Dyn_fly=0)
   n4 = Branch(execute=@n1.then_1, Condition=Get bIsInFrontEnd())
   > Set Dyn_Frontend_alfa(Dyn_Frontend_alfa:float=1) -> (Output_Get:float)
   Set Dyn_Frontend_alfa(execute=@n4.else, Dyn_Frontend_alfa=0)
  ]}
]
variables [
 Dyn_fly {type=number, default=1}
 Dyn_Frontend_alfa {type=number, default=1}
 Dyn_hummer_strap_transform {type=number, default=1}
]
