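# Anim Blueprint for the SteelStomper face accessory, copying the attached parent's pose and adding physics motion to the dyn_hair bones.
# Two additive head-correction bone offsets feed four chained root-relative AnimDynamics nodes with angular springs, then the output pose.
# The update event switches the offsets on only when the current LOD is not 0.
# A second, bone-relative AnimDynamics chain is left unconnected to the output.
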AnimBlueprint /BRCosmetics/Characters/Player/Male/Large/Bodies/M_LRG_SteelStomper/Meshes/Parts/M_LRG_SteelStomper_FaceAcc_AnimBP
target_skeleton=&/M_LRG_SteelStomper_FaceAcc_Skeleton
graphs [
 AnimGraph {nodes [
   n1 = Anim Dynamics - Bone: dyn_hair_2(Alpha=1, ExternalForce="0,0,0") {
    angularDampingOverride=1
    angularSpringConstant=10
    bAngularSpring=true
    bOverrideAngularDamping=true
    bUseGravityOverride=true
    bUsePlanarLimit=false
    linearSpringConstant=1
    simulationSpace=BoneRelative
    alphaBoolBlend {blendOption=Linear}
    alphaScaleBiasClamp {bInterpResult=true, interpSpeedDecreasing=2, interpSpeedIncreasing=2}
    boundBone {boneName=dyn_hair_2}
    componentAppliedLinearAccClamp=<1000 1000 1000>
    componentLinearAccScale=<0.75 0.75 0.75>
    componentLinearVelScale=<0.75 0.75 0.75>
    gravityOverride {z=-30}
    physicsBodyDefinitions [{
      boundBone {boneName=dyn_hair_2}
      boxExtents=<15 20 5>
      constraintSetup {angularTargetAxis=AxisY, coneAngle=90, twistAxis=AxisZ, angularLimitsMax=<30 30 10>, angularLimitsMin=<-30 -30 -10>, angularTarget=<0 0 1>}
      localJointOffset=<0.5 1 0>
     }]
    relativeSpaceBone {boneName=neck_02}
   }
   n2 = Anim Dynamics - Bone: dyn_hair_3(ComponentPose=@n1, Alpha=1, ExternalForce="0,0,0") {
    angularDampingOverride=1
    angularSpringConstant=50
    bAngularSpring=true
    bGravityOverrideInSimSpace=true
    bOverrideAngularDamping=true
    bUseGravityOverride=true
    bUsePlanarLimit=false
    linearSpringConstant=1
    simulationSpace=BoneRelative
    alphaBoolBlend {blendOption=Linear}
    alphaScaleBiasClamp {bInterpResult=true, interpSpeedDecreasing=2, interpSpeedIncreasing=2}
    boundBone {boneName=dyn_hair_3}
    componentAppliedLinearAccClamp=<1000 1000 1000>
    componentLinearAccScale=<0.75 0.75 0.75>
    componentLinearVelScale=<0.75 0.75 0.75>
    gravityOverride {z=-5}
    physicsBodyDefinitions [{
      boundBone {boneName=dyn_hair_3}
      boxExtents=<10 20 5>
      constraintSetup {angularTargetAxis=AxisY, coneAngle=90, twistAxis=AxisZ, angularLimitsMax=<10 45 10>, angularLimitsMin=<-30 -45 -10>, angularTarget=<0 0 1>}
      localJointOffset=<0.5 1 0>
     }]
    relativeSpaceBone {boneName=head}
   }
   Anim Dynamics - Bone: dyn_hair_4(ComponentPose=@n2, Alpha=1, ExternalForce="0,0,0") {
    angularDampingOverride=1
    angularSpringConstant=50
    bAngularSpring=true
    bGravityOverrideInSimSpace=true
    bOverrideAngularDamping=true
    bUseGravityOverride=true
    bUsePlanarLimit=false
    linearSpringConstant=1
    simulationSpace=BoneRelative
    alphaBoolBlend {blendOption=Linear}
    alphaScaleBiasClamp {bInterpResult=true, interpSpeedDecreasing=2, interpSpeedIncreasing=2}
    boundBone {boneName=dyn_hair_4}
    componentAppliedLinearAccClamp=<1000 1000 1000>
    componentLinearAccScale=<0.75 0.75 0.75>
    componentLinearVelScale=<0.75 0.75 0.75>
    gravityOverride {z=-5}
    physicsBodyDefinitions [{
      boundBone {boneName=dyn_hair_4}
      boxExtents=<10 20 5>
      constraintSetup {angularTargetAxis=AxisY, coneAngle=45, twistAxis=AxisZ, angularLimitsMax=<10 45 10>, angularLimitsMin=<-30 -45 -10>, angularTarget=<0 0 1>}
      localJointOffset=<0.5 1 0>
     }]
    relativeSpaceBone {boneName=head}
   }
   n3 = Get Dyn_Transform_Alfa() -> (Dyn_Transform_Alfa:double)
   n4 = "Transform (Modify) Bone - Bone: dyn_head_correct"(
    ComponentPose=Local To Component(LocalPose=Copy Pose From Mesh() {bUseAttachedParent=true})
    Alpha=@n3
    Translation="4.5,0,0"
    Rotation="0,0,0"
    Scale="1,1,1"
   ) {
    rotationMode=BMM_Additive
    rotationSpace=BCS_BoneSpace
    translationMode=BMM_Additive
    translationSpace=BCS_BoneSpace
    alphaBoolBlend {blendOption=Linear}
    alphaScaleBiasClamp {bInterpResult=true}
    boneToModify {boneName=dyn_head_correct}
    translation {x=4.5}
   }
   n5 = "Transform (Modify) Bone - Bone: dyn_head_correct_back"(ComponentPose=@n4, Alpha=@n3, Translation="0,-4,0", Rotation="0,0,0", Scale="1,1,1") {
    rotationMode=BMM_Additive
    rotationSpace=BCS_BoneSpace
    translationMode=BMM_Additive
    translationSpace=BCS_BoneSpace
    alphaBoolBlend {blendOption=Linear}
    alphaScaleBiasClamp {bInterpResult=true}
    boneToModify {boneName=dyn_head_correct_back}
    translation {y=-4}
   }
   n6 = Anim Dynamics - Bone: dyn_hair_1(ComponentPose=@n5, Alpha=0.65, ExternalForce="0,0,0") {
    alpha=0.75
    angularDampingOverride=0.95
    angularSpringConstant=120
    bAngularSpring=true
    bOverrideAngularDamping=true
    bUseGravityOverride=true
    bUsePlanarLimit=false
    linearSpringConstant=1
    simulationSpace=RootRelative
    alphaBoolBlend {blendOption=Linear}
    alphaScaleBiasClamp {bInterpResult=true}
    boundBone {boneName=dyn_hair_1}
    componentAppliedLinearAccClamp=<1000 1000 1000>
    componentLinearAccScale=<0.75 0.75 0.75>
    componentLinearVelScale=<0.85 0.85 0.85>
    gravityOverride {z=30}
    physicsBodyDefinitions [{
      boundBone {boneName=dyn_hair_1}
      boxExtents=<7.5 5 10>
      constraintSetup {angularTargetAxis=AxisZ, coneAngle=75, angularLimitsMax=<30 15 10>, angularLimitsMin=<-30 -15 -10>, angularTarget=<0 0 1>}
      localJointOffset=<0 0 1.25>
     }]
    relativeSpaceBone {boneName=neck_02}
   }
   n7 = Anim Dynamics - Bone: dyn_hair_2(ComponentPose=@n6, Alpha=0.4, ExternalForce="0,0,0") {
    alpha=0.75
    angularDampingOverride=0.95
    angularSpringConstant=100
    bAngularSpring=true
    bOverrideAngularDamping=true
    bUseGravityOverride=true
    bUsePlanarLimit=false
    linearSpringConstant=1
    simulationSpace=RootRelative
    alphaBoolBlend {blendOption=Linear}
    alphaScaleBiasClamp {bInterpResult=true}
    boundBone {boneName=dyn_hair_2}
    componentAppliedLinearAccClamp=<1000 1000 1000>
    componentLinearAccScale=<0.75 0.75 0.75>
    componentLinearVelScale=<0.85 0.85 0.85>
    gravityOverride {z=20}
    physicsBodyDefinitions [{
      boundBone {boneName=dyn_hair_2}
      boxExtents=<7 5 10>
      constraintSetup {angularTargetAxis=AxisZ, coneAngle=75, angularLimitsMax=<30 15 10>, angularLimitsMin=<-30 -15 -10>, angularTarget=<0 0 1>}
      localJointOffset=<0 0 1.15>
     }]
    relativeSpaceBone {boneName=neck_02}
   }
   n8 = Anim Dynamics - Bone: dyn_hair_3(ComponentPose=@n7, Alpha=0.8, ExternalForce="0,0,0") {
    alpha=0.75
    angularDampingOverride=0.8
    angularSpringConstant=15
    bAngularSpring=true
    bOverrideAngularDamping=true
    bUseGravityOverride=true
    bUsePlanarLimit=false
    linearSpringConstant=1
    simulationSpace=RootRelative
    alphaBoolBlend {blendOption=Linear}
    alphaScaleBiasClamp {bInterpResult=true}
    boundBone {boneName=dyn_hair_3}
    componentAppliedLinearAccClamp=<1000 1000 1000>
    componentLinearAccScale=<0.75 0.75 0.75>
    componentLinearVelScale=<0.85 0.85 0.85>
    gravityOverride {z=60}
    physicsBodyDefinitions [{
      boundBone {boneName=dyn_hair_3}
      boxExtents=<10 5 10>
      constraintSetup {angularTargetAxis=AxisZ, coneAngle=75, angularLimitsMax=<30 15 10>, angularLimitsMin=<-30 -15 -10>, angularTarget=<0 0 1>}
      localJointOffset=<0 0 1>
     }]
    relativeSpaceBone {boneName=neck_02}
   }
   n9 = Anim Dynamics - Bone: dyn_hair_4(ComponentPose=@n8, Alpha=0.75, ExternalForce="0,0,0") {
    alpha=0.75
    angularDampingOverride=0.8
    angularSpringConstant=15
    bAngularSpring=true
    bOverrideAngularDamping=true
    bUseGravityOverride=true
    bUsePlanarLimit=false
    linearSpringConstant=1
    simulationSpace=RootRelative
    alphaBoolBlend {blendOption=Linear}
    alphaScaleBiasClamp {bInterpResult=true}
    boundBone {boneName=dyn_hair_4}
    componentAppliedLinearAccClamp=<1000 1000 1000>
    componentLinearAccScale=<0.75 0.75 0.75>
    componentLinearVelScale=<0.85 0.85 0.85>
    gravityOverride {z=50}
    physicsBodyDefinitions [{
      boundBone {boneName=dyn_hair_4}
      boxExtents=<10 5 10>
      constraintSetup {angularTargetAxis=AxisZ, coneAngle=75, angularLimitsMax=<30 15 10>, angularLimitsMin=<-30 -15 -10>, angularTarget=<0 0 1>}
      localJointOffset=<0 0 0.85>
     }]
    relativeSpaceBone {boneName=neck_02}
   }
   Output Pose(Result=Component To Local(ComponentPose=@n9)) {layerGroup=None}
  ]}
 EventGraph {nodes [
   Event BlueprintUpdateAnimation()
   > n1 = Branch(Condition="Not Equal (Integer)"(A=Get CurrentLOD()))
   > Set Dyn_Transform_Alfa(Dyn_Transform_Alfa:double=1) -> (Output_Get:double)
   Set Dyn_Transform_Alfa(execute=@n1.else, Dyn_Transform_Alfa=0)
  ]}
]
variables [Dyn_Transform_Alfa {type=number}]
